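# Niagara eye glow effect for the CopperToll character, pinned each update to the skinned head bone.
# The eyes_light emitter places orange sprites with lights at two mirrored offsets by spawn group, pulsing alpha and size, with alpha also scaled by time of day.
# Middle_Glow adds green sprites at a central head offset that shrink with camera distance; its alpha flicker curve is switched off.
# User parameters, including eye_glow, are declared but unreferenced in the modules shown.
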NiagaraSystem /BRCosmetics/Characters/Player/Male/Medium/Bodies/M_MED_CopperToll/FX/NS_CopperToll_Eye_glow
dependencies {
 usedDataInterfaces [
  &NiagaraDataInterfaceCurve
  &NiagaraDataInterfaceSkeletalMesh
  &NiagaraDataInterfaceArrayFloat3
  &NiagaraDataInterfaceCamera
 ]
 usedDynamicInputs []
 usedModules [&SystemState]
}
emitters {
 Middle_Glow {
  emitterUpdateScript [
   &EmitterState(
    Inactive Response=::NewEnumerator2
    Life Cycle Mode=::NewEnumerator1
    Loop Behavior=::NewEnumerator1
    Loop Duration Mode=::NewEnumerator1
    UseLoopDelay=false
   )
   &SpawnBurst_Instantaneous(Spawn Count=2, Use Spawn Probability=false, UseLoopCountLimit=true)
  ]
  particleSpawnScript [
   &InitializeParticle(
    Color=<rgba 0.0299 0.4896 0.2392 1>
    Sprite Size Mode=::NewEnumerator3
    Uniform Sprite Size=15
    UsePositionOffset=false
    Write Material Random=true
   )
   &DynamicMaterialParameters(
    Index 0 Param 1=3.5
    Index 0 Param 2=0.5
    Index 0 Param 4=100
    Module.Index0Param0Write=true
    Module.Index0Param1Write=true
    Module.Index0Param2Write=true
    Module.Index0Param3Write=true
   )
   &CameraOffset(Camera Offset Amount=4)
  ]
  particleUpdateScript [
   &ParticleState(Kill Particles When Lifetime Has Elapsed=false, Let Infinitely Lived Particles Die When Emitter Deactivates=true)
   &SkeletalMeshLocation(
    Bone Sampling Mode=::NewEnumerator4
    MaskSpawnGroup=false
    Sampled Orientation Coordinate Space=::World
    Sampled Position Coordinate Space=::World
    Sampled Position Offset=<17.5 7 0>
    Sampled Velocity Coordinate Space:ENiagaraCoordinateSpace=::World
    Skeletal Mesh=DataInterface() {propertyValues {
      DefaultMesh=null
      ExcludeBoneName=root
      OptionalComponentClass=null
      PreviewMesh=/BRCosmetics/Characters/Player/Male/Medium/Heads/M_MED_CopperToll_Head/Meshes/M_MED_CopperToll_HeadNew
      SkinningMode=SkinOnTheFly
      SoftSourceActor=null
      SourceMode=Default
      UvSetIndex=0
      WholeMeshLOD=-1
      bExcludeBone=false
      bReadDeformedGeometry=true
      bRequireCurrentFrameData=true
      ComponentTags []
      FilteredBones [head]
      FilteredSockets []
      MeshUserParameter {parameter {name=None, typeDefHandle {registeredTypeIndex=1}, varData []}}
      SamplingRegions []
     }}
   )
   &ScaleColor(
    Scale Alpha=&FloatFromCurve(
     FloatCurve=DataInterface() {propertyValues {Curve {keys [
         <at 0.1 1>
         {arriveTangent=-0.7039, interpMode=RCIM_Cubic, leaveTangent=-0.7039, time=0.3472, value=1}
         <at 0.5972 0.65 RCIM_Cubic>
         <at 0.8472 1>
         <at 1.0972 0.65 RCIM_Cubic>
         <at 1.3472 1 RCIM_Cubic>
         <at 1.5972 0.65>
         <at 1.8472 1 RCIM_Cubic>
        ]}}}
     CurveIndex=$System.NormalizedLoopAge
    )
    ScaleA=false
    ScaleRGB=true
   )
   &ScaleSpriteSize(
    Scale Sprite Size Mode=::NewEnumerator0
    Uniform Scale Factor=&ClampFloat(Float=&Lerp_Float(
      A=1
      B=0.5
      Alpha=&RemapRange(
       Input Value To Remap=&VectorLength(Vector=&Subtract_Position(
         A=/Niagara/DynamicInputs/Helpers/SimulationPosition()
         B=&ReturnCameraPosition(Coordinate Space:ENiagaraCoordinateSpace=::World, Camera Query:NiagaraDataInterfaceCamera=DataInterface())
        ))
       Input Min=500
       Input Max=2000
       Output Min=0
       Output Max=1
      )
     ))
   ) {module=ScaleSpriteSize001}
  ]
  rendererClasses [&NiagaraSpriteRendererProperties]
  renderers [{rendererClass=&NiagaraSpriteRendererProperties}]
 }
 eyes_light {
  emitterUpdateScript [
   &EmitterState(
    Inactive Response=::NewEnumerator1
    Life Cycle Mode=::NewEnumerator1
    Loop Behavior=::NewEnumerator1
    Loop Duration Mode=::NewEnumerator1
    Reset Age On Awaken=true
    Scalability Mode=::NewEnumerator1
    UseLoopDelay=false
   )
   &SpawnBurst_Instantaneous(Spawn Count=2, Use Spawn Probability=false)
  ]
  particleSpawnScript [
   &FN_SetSpawnGroup(Number of Spawn Groups=2)
   &InitializeParticle(Color=<rgba 1 0.4802 0 0.25>, Sprite Size Mode=::NewEnumerator3, Uniform Sprite Size=17, UsePositionOffset=false, Write Material Random=true) {module=InitializeParticle001}
   &CameraOffset(Camera Offset Amount=1.1)
  ]
  particleUpdateScript [
   &ParticleState(Kill Particles When Lifetime Has Elapsed=false, Let Infinitely Lived Particles Die When Emitter Deactivates=true)
   &ScaleColor(
    Scale Alpha=&Multiply_Float(
     A=&Multiply_Float(A=1, B=&FNTimeOfDayScalar(Night Scale=0.8, Morning Scale=0.3, Evening Scale=0.5))
     B=&FloatFromCurve(
      FloatCurve=DataInterface() {propertyValues {Curve {keys [<at 0 1>,<at 0.0539 1>,<at 0.1557 0.8>,<at 0.2614 1>,<at 0.3742 0.8>,<at 0.456 1>,<at 0.5872 0.8>,<at 0.7535 1>,<at 0.8976 0.8>,<at 1 1>]}}}
      CurveIndex=$System.NormalizedLoopAge
     )
    )
    ScaleA=true
    ScaleRGB=true
   )
   &SolveForcesAndVelocity(Clamp Velocity=false, Limit Acceleration=false)
   &SkeletalMeshLocation(
    Apply Sampled Orientation to Mesh Particles=true
    Bone / Socket Index=$Particles.SpawnGroup
    Bone Sampling Mode=::NewEnumerator4
    Inherit Velocity=false
    Interpolate Mesh Sampling=false
    MaskSpawnGroup=false
    Sampled Position Offset=&SelectVectorFromArray(Direct Array Index=$Particles.SpawnGroup, Vector Selection Array=DataInterface() {propertyValues {FloatData [<5 8 3.2>, <5 8 -3.2>]}})
    Sampled Position Offset Coordinate Space:ENiagaraMeshTransforms=::NewEnumerator3
    Skeletal Mesh=DataInterface() {propertyValues {
      DefaultMesh=null
      ExcludeBoneName=root
      OptionalComponentClass=null
      PreviewMesh=&M_MED_CopperToll_HeadNew
      SkinningMode=SkinOnTheFly
      SoftSourceActor=null
      SourceMode=Default
      UvSetIndex=0
      WholeMeshLOD=-1
      bExcludeBone=false
      bReadDeformedGeometry=true
      bRequireCurrentFrameData=true
      ComponentTags []
      FilteredBones [head,head]
      FilteredSockets []
      MeshUserParameter {parameter {name=None, typeDefHandle {registeredTypeIndex=1}, varData []}}
      SamplingRegions []
     }}
   ) {module=SkeletalMeshLocation001}
   &DynamicMaterialParameters(
    Index 0 Param 1=1
    Index 0 Param 2=1
    Index 0 Param 3=1
    Index 0 Param 4=1
    Index 1 Param 1=2
    Index 1 Param 2=1
    Index 1 Param 3=1
    Index 1 Param 4=1
    Module.Param0WriteEnabled=true
    Module.Param0WriteEnabled1=true
    Module.Param1WriteEnabled=true
    Module.Param1WriteEnabled1=true
    Module.Param2WriteEnabled=true
    Module.Param2WriteEnabled1=true
    Module.Param3WriteEnabled=true
    Module.Param3WriteEnabled1=true
    Write Parameter Index 1=true
   )
   &ScaleSpriteSize(Scale Factor=&Vector2DFromFloat(Value=&FloatFromCurve(
      FloatCurve=DataInterface() {propertyValues {Curve {keys [<at 0 1>,<at 0.0539 1>,<at 0.1557 0.985>,<at 0.2614 1>,<at 0.3742 0.985>,<at 0.456 1>,<at 0.5872 0.985>,<at 0.7535 1>,<at 0.8976 0.985>,<at 1 1>]}}}
      CurveIndex=$System.NormalizedLoopAge
     )))
   &FN_Light_Attributes(Color=<rgba 6 2.8811 0 1>, Create Color Attribute=true, Create Exponent Attribute=true, Create Radius Attribute=true, Exponent=2, Radius=10)
  ]
  rendererClasses [&NiagaraSpriteRendererProperties, &NiagaraLightRendererProperties]
  renderers [
   {rendererClass=&NiagaraSpriteRendererProperties}
   {rendererClass=&NiagaraLightRendererProperties, rendererIndex=1}
  ]
 }
}
summary {
 systemName=NS_CopperToll_Eye_glow
 emitters [
  {emitterName=eyes_light, rendererClasses [&NiagaraSpriteRendererProperties, &NiagaraLightRendererProperties]}
  {emitterName=Middle_Glow, rendererClasses [&NiagaraSpriteRendererProperties]}
 ]
 userVariables [
  User.bIsFrontEnd {type=NiagaraBool}
  User.bIsFrontEndPreview {type=NiagaraBool}
  User.eye_glow {type=NiagaraFloat, defaultValue=0, description=""}
 ]
}
