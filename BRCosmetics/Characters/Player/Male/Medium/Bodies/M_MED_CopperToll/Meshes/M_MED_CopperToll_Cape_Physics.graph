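# Physics asset for the cape of the male medium CopperToll cosmetic body, covering its sleeve, back and necklace dynamic bones.
# Each chain starts at a tiny MSTR anchor capsule with free rotation on spine_05, followed by capsule links with tight swing limits and zero twist limit.
# Sleeve chains are asymmetric: four single-link chains on the left, three two-link chains on the right.
# Larger spine and upper-arm capsules, joined by free constraints, provide the torso and arm volumes.
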PhysicsAsset /BRCosmetics/Characters/Player/Male/Medium/Bodies/M_MED_CopperToll/Meshes/M_MED_CopperToll_Cape_Physics
bodies {
 dyn_capeSleeveA_l_01 [{shapeType=Capsule, length=6, radius=4.5, shapeName=dyn_capeSleeveA_l_01_capsule, center=<3.0408 0 4>, rotation=<rot 90 0 0>}]
 dyn_capeSleeveA_l_MSTR [{shapeType=Capsule, length=0.5, radius=0.5, shapeName=dyn_capeSleeveA_l_MSTR_capsule, rotation=<rot 0 -90 0>}]
 dyn_capeSleeveA_r_01 [{shapeType=Capsule, length=6, radius=4.5, shapeName=dyn_capeSleeveA_r_01_capsule, center=<3.0408 0 4>, rotation=<rot 90 0 0>}]
 dyn_capeSleeveA_r_02 [{shapeType=Capsule, length=6, radius=4.5, shapeName=dyn_capeSleeveA_r_02_capsule, center=<3.0408 0 4>, rotation=<rot 90 0 0>}]
 dyn_capeSleeveA_r_MSTR [{shapeType=Capsule, length=0.5, radius=0.5, shapeName=dyn_capeSleeveA_r_MSTR_capsule, rotation=<rot 0 -90 0>}]
 dyn_capeSleeveB_l_01 [{shapeType=Capsule, length=6, radius=4.5, shapeName=dyn_capeSleeveB_l_01_capsule, center=<3.0408 0 4>, rotation=<rot 90 0 0>}]
 dyn_capeSleeveB_l_MSTR [{shapeType=Capsule, length=0.5, radius=0.5, shapeName=dyn_capeSleeveB_l_MSTR_capsule, rotation=<rot 0 -90 0>}]
 dyn_capeSleeveB_r_01 [{shapeType=Capsule, length=6, radius=4.5, shapeName=dyn_capeSleeveB_r_01_capsule, center=<3.0408 0 4>, rotation=<rot 90 0 0>}]
 dyn_capeSleeveB_r_02 [{shapeType=Capsule, length=6, radius=4.5, shapeName=dyn_capeSleeveB_r_02_capsule, center=<3.0408 0 4>, rotation=<rot 90 0 0>}]
 dyn_capeSleeveB_r_MSTR [{shapeType=Capsule, length=0.5, radius=0.5, shapeName=dyn_capeSleeveB_r_MSTR_capsule, rotation=<rot 0 -90 0>}]
 dyn_capeSleeveBack_c_01 [{shapeType=Capsule, length=6.5, radius=4.5, shapeName=dyn_capeSleeveBack_c_01_capsule, center=<3.0408 0 4>, rotation=<rot 90 0 0>}]
 dyn_capeSleeveBack_c_02 [{shapeType=Capsule, length=13.5, radius=5, shapeName=dyn_capeSleeveBack_c_02_capsule, center=<3.0408 0 4>, rotation=<rot 75.2667 0 0>}]
 dyn_capeSleeveBack_c_MSTR [{shapeType=Capsule, length=0.5, radius=0.5, shapeName=dyn_capeSleeveBack_c_MSTR_capsule, rotation=<rot 0 -90 0>}]
 dyn_capeSleeveC_l_01 [{shapeType=Capsule, length=6, radius=4.5, shapeName=dyn_capeSleeveC_l_01_capsule, center=<3.0408 0 4>, rotation=<rot 90 0 0>}]
 dyn_capeSleeveC_l_MSTR [{shapeType=Capsule, length=0.5, radius=0.5, shapeName=dyn_capeSleeveC_l_MSTR_capsule, rotation=<rot 0 -90 0>}]
 dyn_capeSleeveC_r_01 [{shapeType=Capsule, length=6, radius=4.5, shapeName=dyn_capeSleeveC_r_01_capsule, center=<3.0408 0 4>, rotation=<rot 90 0 0>}]
 dyn_capeSleeveC_r_02 [{shapeType=Capsule, length=6, radius=4.5, shapeName=dyn_capeSleeveC_r_02_capsule, center=<3.0408 0 4>, rotation=<rot 90 0 0>}]
 dyn_capeSleeveC_r_MSTR [{shapeType=Capsule, length=0.5, radius=0.5, shapeName=dyn_capeSleeveC_r_MSTR_capsule, rotation=<rot 0 -90 0>}]
 dyn_capeSleeveD_l_01 [{shapeType=Capsule, length=6, radius=4.5, shapeName=dyn_capeSleeveD_l_01_capsule, center=<3.0408 0 4>, rotation=<rot 90 0 0>}]
 dyn_capeSleeveD_l_MSTR [{shapeType=Capsule, length=0.5, radius=0.5, shapeName=dyn_capeSleeveD_l_MSTR_capsule, rotation=<rot 0 -90 0>}]
 dyn_necklace_01 [{shapeType=Capsule, length=9.5, radius=4, shapeName=dyn_necklace_01_capsule, center=<5.1856 0 4.5>, rotation=<rot 0 0 90>}]
 dyn_necklace_02 [{shapeType=Capsule, length=4.5, radius=5.5, shapeName=dyn_necklace_02_capsule, center=<5.1856 0 4.5>, rotation=<rot 0 0 90>}]
 dyn_necklace_03 [{shapeType=Capsule, length=4.5, radius=6.5, shapeName=dyn_necklace_03_capsule, center=<8.2154 0 4.5>, rotation=<rot 0 0 90>}]
 dyn_necklace_MSTR [{shapeType=Capsule, length=0.5, radius=0.5, shapeName=dyn_necklace_MSTR_capsule}]
 spine_03 [{shapeType=Capsule, length=8, radius=11.5, shapeName=spine_03_capsule, center=<5.2399 -0.2247 0>}]
 spine_04 [{shapeType=Capsule, length=8, radius=12, shapeName=spine_04_capsule, center=<3.3049 0 0>}]
 spine_05 [{shapeType=Capsule, length=13.5, radius=12, shapeName=spine_05_capsule}]
 upperarm_l [{shapeType=Capsule, length=32.5711, radius=4.569, shapeName=upperarm_l_capsule, center=<14.2257 0 0>, rotation=<rot 90 0 0>}]
 upperarm_r [{shapeType=Capsule, length=32.5711, radius=4.569, shapeName=upperarm_l_capsule, center=<-14.2257 0 0>, rotation=<rot 90 0 180>}]
}
constraints [
 {
  bone1Name=dyn_capeSleeveA_l_MSTR
  bone2Name=spine_05
  swing1LimitDegrees=45
  swing1Motion=Free
  swing2LimitDegrees=45
  swing2Motion=Free
  twistLimitDegrees=45
  twistMotion=Free
 }
 {bone1Name=dyn_capeSleeveA_l_01, bone2Name=dyn_capeSleeveA_l_MSTR, swing1LimitDegrees=10, swing2LimitDegrees=10, twistLimitDegrees=0}
 {
  bone1Name=dyn_capeSleeveB_l_MSTR
  bone2Name=spine_05
  swing1LimitDegrees=45
  swing1Motion=Free
  swing2LimitDegrees=45
  swing2Motion=Free
  twistLimitDegrees=45
  twistMotion=Free
 }
 {bone1Name=dyn_capeSleeveB_l_01, bone2Name=dyn_capeSleeveB_l_MSTR, swing1LimitDegrees=30, swing2LimitDegrees=30, twistLimitDegrees=0}
 {
  bone1Name=dyn_capeSleeveC_l_MSTR
  bone2Name=spine_05
  swing1LimitDegrees=45
  swing1Motion=Free
  swing2LimitDegrees=45
  swing2Motion=Free
  twistLimitDegrees=45
  twistMotion=Free
 }
 {bone1Name=dyn_capeSleeveC_l_01, bone2Name=dyn_capeSleeveC_l_MSTR, swing1LimitDegrees=15, swing2LimitDegrees=40, twistLimitDegrees=0}
 {
  bone1Name=dyn_capeSleeveA_r_MSTR
  bone2Name=spine_05
  swing1LimitDegrees=45
  swing1Motion=Free
  swing2LimitDegrees=45
  swing2Motion=Free
  twistLimitDegrees=45
  twistMotion=Free
 }
 {bone1Name=dyn_capeSleeveA_r_01, bone2Name=dyn_capeSleeveA_r_MSTR, swing1LimitDegrees=10, swing2LimitDegrees=10, twistLimitDegrees=0}
 {bone1Name=dyn_capeSleeveA_r_02, bone2Name=dyn_capeSleeveA_r_01, swing1LimitDegrees=10, swing2LimitDegrees=10, twistLimitDegrees=0}
 {
  bone1Name=dyn_capeSleeveB_r_MSTR
  bone2Name=spine_05
  swing1LimitDegrees=45
  swing1Motion=Free
  swing2LimitDegrees=45
  swing2Motion=Free
  twistLimitDegrees=45
  twistMotion=Free
 }
 {bone1Name=dyn_capeSleeveB_r_01, bone2Name=dyn_capeSleeveB_r_MSTR, swing1LimitDegrees=10, swing2LimitDegrees=15, twistLimitDegrees=0}
 {bone1Name=dyn_capeSleeveB_r_02, bone2Name=dyn_capeSleeveB_r_01, swing1LimitDegrees=10, swing2LimitDegrees=15, twistLimitDegrees=0}
 {
  bone1Name=dyn_capeSleeveC_r_MSTR
  bone2Name=spine_05
  swing1LimitDegrees=45
  swing1Motion=Free
  swing2LimitDegrees=45
  swing2Motion=Free
  twistLimitDegrees=45
  twistMotion=Free
 }
 {bone1Name=dyn_capeSleeveC_r_01, bone2Name=dyn_capeSleeveC_r_MSTR, swing1LimitDegrees=15, swing2LimitDegrees=17, twistLimitDegrees=0}
 {bone1Name=dyn_capeSleeveC_r_02, bone2Name=dyn_capeSleeveC_r_01, swing1LimitDegrees=5, swing2LimitDegrees=17, twistLimitDegrees=0}
 {
  bone1Name=dyn_capeSleeveBack_c_MSTR
  bone2Name=spine_05
  swing1LimitDegrees=45
  swing1Motion=Free
  swing2LimitDegrees=45
  swing2Motion=Free
  twistLimitDegrees=45
  twistMotion=Free
 }
 {bone1Name=dyn_capeSleeveBack_c_01, bone2Name=dyn_capeSleeveBack_c_MSTR, swing1LimitDegrees=30, swing2LimitDegrees=30, twistLimitDegrees=0}
 {bone1Name=dyn_capeSleeveBack_c_02, bone2Name=dyn_capeSleeveBack_c_01, swing1LimitDegrees=30, swing2LimitDegrees=30, twistLimitDegrees=0}
 {
  bone1Name=dyn_capeSleeveD_l_MSTR
  bone2Name=spine_05
  swing1LimitDegrees=45
  swing1Motion=Free
  swing2LimitDegrees=45
  swing2Motion=Free
  twistLimitDegrees=45
  twistMotion=Free
 }
 {bone1Name=dyn_capeSleeveD_l_01, bone2Name=dyn_capeSleeveD_l_MSTR, swing1LimitDegrees=30, swing2LimitDegrees=30, twistLimitDegrees=0}
 {
  bone1Name=upperarm_r
  bone2Name=spine_05
  swing1LimitDegrees=45
  swing1Motion=Free
  swing2LimitDegrees=45
  swing2Motion=Free
  twistLimitDegrees=45
  twistMotion=Free
 }
 {
  bone1Name=spine_05
  bone2Name=spine_04
  swing1LimitDegrees=45
  swing1Motion=Free
  swing2LimitDegrees=45
  swing2Motion=Free
  twistLimitDegrees=45
  twistMotion=Free
 }
 {
  bone1Name=spine_04
  bone2Name=spine_03
  swing1LimitDegrees=45
  swing1Motion=Free
  swing2LimitDegrees=45
  swing2Motion=Free
  twistLimitDegrees=45
  twistMotion=Free
 }
 {
  bone1Name=upperarm_l
  bone2Name=spine_05
  swing1LimitDegrees=45
  swing1Motion=Free
  swing2LimitDegrees=45
  swing2Motion=Free
  twistLimitDegrees=45
  twistMotion=Free
 }
 {
  bone1Name=dyn_necklace_MSTR
  bone2Name=spine_05
  swing1LimitDegrees=45
  swing1Motion=Free
  swing2LimitDegrees=45
  swing2Motion=Free
  twistLimitDegrees=45
  twistMotion=Free
 }
 {bone1Name=dyn_necklace_01, bone2Name=dyn_necklace_MSTR, swing1LimitDegrees=5, swing2LimitDegrees=20, twistLimitDegrees=0}
 {bone1Name=dyn_necklace_02, bone2Name=dyn_necklace_01, swing1LimitDegrees=20, swing2LimitDegrees=20, twistLimitDegrees=0}
 {bone1Name=dyn_necklace_03, bone2Name=dyn_necklace_02, swing1LimitDegrees=0, swing2LimitDegrees=20, twistLimitDegrees=0}
]
