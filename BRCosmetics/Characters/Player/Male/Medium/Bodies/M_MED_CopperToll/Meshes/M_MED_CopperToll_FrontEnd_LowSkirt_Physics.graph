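# Physics asset for the CopperToll medium male body's front-end low skirt, providing secondary motion for cloth panels and a left pouch.
# Six cloth capsule chains and the pouch chain hang from free-jointed MSTR anchors on the pelvis, with swing-limited links, zero twist for cloth and some twist for the pouch.
# Pelvis and leg capsules form body shapes, and a huge sphere-like root capsule sits below as an apparent floor.
# dyn_cloth_r_04 has a second, free constraint to r_03.
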PhysicsAsset /BRCosmetics/Characters/Player/Male/Medium/Bodies/M_MED_CopperToll/Meshes/M_MED_CopperToll_FrontEnd_LowSkirt_Physics
bodies {
 calf_l [{shapeType=Capsule, length=33.5, radius=8.5, shapeName=calf_r_capsule, center=<-26.7833 -1.2685 0>, rotation=<rot 90 7.125 -164.3982>}]
 calf_r [{shapeType=Capsule, length=33.5, radius=8.5, shapeName=calf_r_capsule, center=<26.7833 1.2685 0>, rotation=<rot 90 83.2785 91.8225>}]
 dyn_cloth_bk_01 [{shapeType=Capsule, length=10, radius=5.5, shapeName=dyn_cloth_bk_01_capsule, center=<6.0689 0 5>, rotation=<rot 0 179.972 90>}]
 dyn_cloth_bk_02 [{shapeType=Capsule, length=9, radius=6.5, shapeName=dyn_cloth_bk_02_capsule, center=<6.0689 0 5>, rotation=<rot 0 0 90>}]
 dyn_cloth_bk_03 [{shapeType=Capsule, length=8.5, radius=7, shapeName=dyn_cloth_bk_03_capsule, center=<6.0689 0 5>, rotation=<rot 0 0 90>}]
 dyn_cloth_bk_04 [{shapeType=Capsule, length=14, radius=7, shapeName=dyn_cloth_bk_04_capsule, center=<6.0689 0 5>, rotation=<rot 0 0 90>}]
 dyn_cloth_bk_05 [{shapeType=Capsule, length=14, radius=7, shapeName=dyn_cloth_bk_05_capsule, center=<6.0689 0 5>, rotation=<rot 0 0 90>}]
 dyn_cloth_bk_MSTR [{shapeType=Capsule, length=0.5, radius=0.5, shapeName=dyn_cloth_bk_MSTR_capsule}]
 dyn_cloth_bk_l_01 [{shapeType=Capsule, length=16.5, radius=2.5, shapeName=dyn_cloth_bk_l_01_capsule, center=<6.0689 0 5>, rotation=<rot 0 166.7387 90>}]
 dyn_cloth_bk_l_02 [{shapeType=Capsule, length=15.5, radius=3.5, shapeName=dyn_cloth_bk_l_02_capsule, center=<6.0689 0 5>, rotation=<rot 0 0 90>}]
 dyn_cloth_bk_l_03 [{shapeType=Capsule, length=14.5, radius=6, shapeName=dyn_cloth_bk_l_03_capsule, center=<6.0689 0 5>, rotation=<rot 0 0 90>}]
 dyn_cloth_bk_l_04 [{shapeType=Capsule, length=16, radius=7, shapeName=dyn_cloth_bk_l_04_capsule, center=<6.0689 0 5>, rotation=<rot 0 0 90>}]
 dyn_cloth_bk_l_05 [{shapeType=Capsule, length=16, radius=7, shapeName=dyn_cloth_bk_l_05_capsule, center=<6.0689 0 5>, rotation=<rot 0 0 90>}]
 dyn_cloth_bk_l_MSTR [{shapeType=Capsule, length=0.5, radius=0.5, shapeName=dyn_cloth_bk_l_MSTR_capsule}]
 dyn_cloth_fr_l_01 [{shapeType=Capsule, length=8.5, radius=4.75, shapeName=dyn_cloth_fr_l_01_capsule, center=<8.5035 0 5>, rotation=<rot 90 0 0>}]
 dyn_cloth_fr_l_02 [{shapeType=Capsule, length=10.5, radius=5, shapeName=dyn_cloth_fr_l_02_capsule, center=<6.0689 0 5>, rotation=<rot 83.173 -0.0002 -0.0002>}]
 dyn_cloth_fr_l_03 [{shapeType=Capsule, length=13.5, radius=6.5, shapeName=dyn_cloth_fr_l_03_capsule, center=<6.0689 0 5>, rotation=<rot 83.2153 0 0>}]
 dyn_cloth_fr_l_MSTR [{shapeType=Capsule, length=0.5, radius=0.5, shapeName=dyn_cloth_fr_l_MSTR_capsule}]
 dyn_cloth_fr_r_01 [{shapeType=Capsule, length=9.5, radius=6, shapeName=dyn_cloth_fr_r_01_capsule, center=<8.5035 0 5>, rotation=<rot 90 0 0>}]
 dyn_cloth_fr_r_02 [{shapeType=Capsule, length=9.5, radius=6, shapeName=dyn_cloth_fr_r_02_capsule, center=<6.0689 0 5>, rotation=<rot 90 0 0>}]
 dyn_cloth_fr_r_03 [{shapeType=Capsule, length=13, radius=7, shapeName=dyn_cloth_fr_r_03_capsule, center=<6.0689 0 5>, rotation=<rot 90 0 0>}]
 dyn_cloth_fr_r_MSTR [{shapeType=Capsule, length=0.5, radius=0.5, shapeName=dyn_cloth_fr_r_MSTR_capsule}]
 dyn_cloth_l_01 [{shapeType=Capsule, length=6.5, radius=3.5, shapeName=dyn_cloth_l_01_capsule, center=<6.0689 0 5>, rotation=<rot 0 0 90>}]
 dyn_cloth_l_02 [{shapeType=Capsule, length=6, radius=4, shapeName=dyn_cloth_l_02_capsule, center=<6.0689 0 5>, rotation=<rot 0 0 90>}]
 dyn_cloth_l_03 [{shapeType=Capsule, length=7.5, radius=5.5, shapeName=dyn_cloth_l_03_capsule, center=<9.7745 0 5>, rotation=<rot 0 0 90>}]
 dyn_cloth_l_04 [{shapeType=Capsule, length=13, radius=5.5, shapeName=dyn_cloth_l_04_capsule, center=<6.0689 0 5>, rotation=<rot 0 0 90>}]
 dyn_cloth_l_05 [{shapeType=Capsule, length=13, radius=5.5, shapeName=dyn_cloth_l_05_capsule, center=<6.0689 0 5>, rotation=<rot 0 0 90>}]
 dyn_cloth_l_06 [{shapeType=Capsule, length=13, radius=5.5, shapeName=dyn_cloth_l_06_capsule, center=<6.0689 0 5>, rotation=<rot 0 0 90>}]
 dyn_cloth_l_MSTR [{shapeType=Capsule, length=0.5, radius=0.5, shapeName=dyn_cloth_l_MSTR_capsule}]
 dyn_cloth_r_01 [{shapeType=Capsule, length=5, radius=5, shapeName=dyn_cloth_r_01_capsule, center=<6.0689 0 5>, rotation=<rot 0 180 90>}]
 dyn_cloth_r_02 [{shapeType=Capsule, length=4, radius=6, shapeName=dyn_cloth_r_02_capsule, center=<6.0689 0 5>, rotation=<rot 0 0 90>}]
 dyn_cloth_r_03 [{shapeType=Capsule, length=4, radius=6, shapeName=dyn_cloth_r_03_capsule, center=<6.0689 0 5>, rotation=<rot 0 0 90>}]
 dyn_cloth_r_04 [{shapeType=Capsule, length=4, radius=6, shapeName=dyn_cloth_r_04_capsule, center=<6.0689 0 5>, rotation=<rot 0 0 90>}]
 dyn_cloth_r_MSTR [{shapeType=Capsule, length=0.5, radius=0.5, shapeName=dyn_cloth_r_MSTR_capsule}]
 dyn_pouch_l_01 [{shapeType=Capsule, length=14.5, radius=5, shapeName=dyn_pouch_l_01_capsule, center=<2.353 0 6>, rotation=<rot 0 0 90>}]
 dyn_pouch_l_02 [{shapeType=Capsule, length=17, radius=10, shapeName=dyn_pouch_l_02_capsule, center=<6.9941 0 11>, rotation=<rot 0 0 90>}]
 dyn_pouch_l_MSTR [{shapeType=Capsule, length=0.5, radius=0.5, shapeName=dyn_pouch_l_MSTR_capsule}]
 pelvis [{shapeType=Capsule, length=6, radius=10.5, shapeName=pelvis_capsule, center=<0 -2.3052 -0.4969>}]
 root [{shapeType=Capsule, length=0, radius=350, shapeName=root_capsule, center=<0 0 -343.4203>, rotation=<rot 0 -90 0>}]
 thigh_l [{shapeType=Capsule, length=54.5, radius=9, shapeName=thigh_l_capsule, center=<-21.3869 -0.8081 1.8256>, rotation=<rot 90 0 0>}]
 thigh_r [{shapeType=Capsule, length=53.5, radius=9, shapeName=thigh_l_capsule, center=<20.7844 0.3726 0>, rotation=<rot 90 179.8508 1.2361>}]
}
constraints [
 {
  bone1Name=dyn_cloth_bk_MSTR
  bone2Name=pelvis
  swing1LimitDegrees=45
  swing1Motion=Free
  swing2LimitDegrees=45
  swing2Motion=Free
  twistLimitDegrees=45
  twistMotion=Free
 }
 {bone1Name=dyn_cloth_bk_01, bone2Name=dyn_cloth_bk_MSTR, swing1LimitDegrees=15, swing2LimitDegrees=20, twistLimitDegrees=0}
 {bone1Name=dyn_cloth_bk_02, bone2Name=dyn_cloth_bk_01, swing1LimitDegrees=15, swing2LimitDegrees=20, twistLimitDegrees=0}
 {bone1Name=dyn_cloth_bk_03, bone2Name=dyn_cloth_bk_02, swing1LimitDegrees=15, swing2LimitDegrees=25, twistLimitDegrees=0}
 {bone1Name=dyn_cloth_bk_04, bone2Name=dyn_cloth_bk_03, swing1LimitDegrees=15, swing2LimitDegrees=25, twistLimitDegrees=0}
 {bone1Name=dyn_cloth_bk_05, bone2Name=dyn_cloth_bk_04, swing1LimitDegrees=30, swing2LimitDegrees=30, twistLimitDegrees=0}
 {
  bone1Name=dyn_cloth_bk_l_MSTR
  bone2Name=pelvis
  swing1LimitDegrees=45
  swing1Motion=Free
  swing2LimitDegrees=45
  swing2Motion=Free
  twistLimitDegrees=45
  twistMotion=Free
 }
 {bone1Name=dyn_cloth_bk_l_01, bone2Name=dyn_cloth_bk_l_MSTR, swing1LimitDegrees=15, swing2LimitDegrees=20, twistLimitDegrees=0}
 {bone1Name=dyn_cloth_bk_l_02, bone2Name=dyn_cloth_bk_l_01, swing1LimitDegrees=15, swing2LimitDegrees=20, twistLimitDegrees=0}
 {bone1Name=dyn_cloth_bk_l_03, bone2Name=dyn_cloth_bk_l_02, swing1LimitDegrees=15, swing2LimitDegrees=20, twistLimitDegrees=0}
 {bone1Name=dyn_cloth_bk_l_04, bone2Name=dyn_cloth_bk_l_03, swing1LimitDegrees=15, swing2LimitDegrees=20, twistLimitDegrees=0}
 {bone1Name=dyn_cloth_bk_l_05, bone2Name=dyn_cloth_bk_l_04, swing1LimitDegrees=30, swing2LimitDegrees=30, twistLimitDegrees=0}
 {
  bone1Name=dyn_cloth_fr_l_MSTR
  bone2Name=pelvis
  swing1LimitDegrees=45
  swing1Motion=Free
  swing2LimitDegrees=45
  swing2Motion=Free
  twistLimitDegrees=45
  twistMotion=Free
 }
 {bone1Name=dyn_cloth_fr_l_01, bone2Name=dyn_cloth_fr_l_MSTR, swing1LimitDegrees=8, swing2LimitDegrees=45, twistLimitDegrees=0}
 {bone1Name=dyn_cloth_fr_l_02, bone2Name=dyn_cloth_fr_l_01, swing1LimitDegrees=8, swing2LimitDegrees=45, twistLimitDegrees=0}
 {bone1Name=dyn_cloth_fr_l_03, bone2Name=dyn_cloth_fr_l_02, swing1LimitDegrees=8, swing2LimitDegrees=45, twistLimitDegrees=0}
 {
  bone1Name=dyn_cloth_fr_r_MSTR
  bone2Name=pelvis
  swing1LimitDegrees=45
  swing1Motion=Free
  swing2LimitDegrees=45
  swing2Motion=Free
  twistLimitDegrees=45
  twistMotion=Free
 }
 {bone1Name=dyn_cloth_fr_r_01, bone2Name=dyn_cloth_fr_r_MSTR, swing1LimitDegrees=8, swing2LimitDegrees=45, twistLimitDegrees=0}
 {bone1Name=dyn_cloth_fr_r_02, bone2Name=dyn_cloth_fr_r_01, swing1LimitDegrees=8, swing2LimitDegrees=45, twistLimitDegrees=0}
 {bone1Name=dyn_cloth_fr_r_03, bone2Name=dyn_cloth_fr_r_02, swing1LimitDegrees=8, swing2LimitDegrees=45, twistLimitDegrees=0}
 {
  bone1Name=dyn_cloth_l_MSTR
  bone2Name=pelvis
  swing1LimitDegrees=45
  swing1Motion=Free
  swing2LimitDegrees=45
  swing2Motion=Free
  twistLimitDegrees=45
  twistMotion=Free
 }
 {bone1Name=dyn_cloth_l_01, bone2Name=dyn_cloth_l_MSTR, swing1LimitDegrees=30, swing2LimitDegrees=30, twistLimitDegrees=0}
 {bone1Name=dyn_cloth_l_02, bone2Name=dyn_cloth_l_01, swing1LimitDegrees=30, swing2LimitDegrees=30, twistLimitDegrees=0}
 {bone1Name=dyn_cloth_l_03, bone2Name=dyn_cloth_l_02, swing1LimitDegrees=30, swing2LimitDegrees=30, twistLimitDegrees=0}
 {bone1Name=dyn_cloth_l_04, bone2Name=dyn_cloth_l_03, swing1LimitDegrees=30, swing2LimitDegrees=30, twistLimitDegrees=0}
 {bone1Name=dyn_cloth_l_05, bone2Name=dyn_cloth_l_04, swing1LimitDegrees=30, swing2LimitDegrees=30, twistLimitDegrees=0}
 {bone1Name=dyn_cloth_l_06, bone2Name=dyn_cloth_l_05, swing1LimitDegrees=30, swing2LimitDegrees=30, twistLimitDegrees=0}
 {
  bone1Name=dyn_cloth_r_MSTR
  bone2Name=pelvis
  swing1LimitDegrees=45
  swing1Motion=Free
  swing2LimitDegrees=45
  swing2Motion=Free
  twistLimitDegrees=45
  twistMotion=Free
 }
 {bone1Name=dyn_cloth_r_01, bone2Name=dyn_cloth_r_MSTR, swing1LimitDegrees=15, swing2LimitDegrees=20, twistLimitDegrees=0}
 {bone1Name=dyn_cloth_r_02, bone2Name=dyn_cloth_r_01, swing1LimitDegrees=15, swing2LimitDegrees=20, twistLimitDegrees=0}
 {bone1Name=dyn_cloth_r_03, bone2Name=dyn_cloth_r_02, swing1LimitDegrees=15, swing2LimitDegrees=20, twistLimitDegrees=0}
 {bone1Name=dyn_cloth_r_04, bone2Name=dyn_cloth_r_03, swing1LimitDegrees=15, swing2LimitDegrees=20, twistLimitDegrees=0}
 {
  bone1Name=pelvis
  bone2Name=root
  swing1LimitDegrees=45
  swing1Motion=Free
  swing2LimitDegrees=45
  swing2Motion=Free
  twistLimitDegrees=45
  twistMotion=Free
 }
 {
  bone1Name=thigh_l
  bone2Name=pelvis
  swing1LimitDegrees=45
  swing1Motion=Free
  swing2LimitDegrees=45
  swing2Motion=Free
  twistLimitDegrees=45
  twistMotion=Free
 }
 {
  bone1Name=thigh_r
  bone2Name=pelvis
  swing1LimitDegrees=45
  swing1Motion=Free
  swing2LimitDegrees=45
  swing2Motion=Free
  twistLimitDegrees=45
  twistMotion=Free
 }
 {
  bone1Name=calf_r
  bone2Name=thigh_r
  swing1LimitDegrees=45
  swing1Motion=Free
  swing2LimitDegrees=45
  swing2Motion=Free
  twistLimitDegrees=45
  twistMotion=Free
 }
 {
  bone1Name=calf_l
  bone2Name=thigh_l
  swing1LimitDegrees=45
  swing1Motion=Free
  swing2LimitDegrees=45
  swing2Motion=Free
  twistLimitDegrees=45
  twistMotion=Free
 }
 {
  bone1Name=dyn_cloth_r_04
  bone2Name=dyn_cloth_r_03
  swing1LimitDegrees=45
  swing1Motion=Free
  swing2LimitDegrees=45
  swing2Motion=Free
  twistLimitDegrees=45
  twistMotion=Free
 }
 {
  bone1Name=dyn_pouch_l_MSTR
  bone2Name=pelvis
  swing1LimitDegrees=45
  swing1Motion=Free
  swing2LimitDegrees=45
  swing2Motion=Free
  twistLimitDegrees=45
  twistMotion=Free
 }
 {bone1Name=dyn_pouch_l_01, bone2Name=dyn_pouch_l_MSTR, swing1LimitDegrees=30, swing2LimitDegrees=20, twistLimitDegrees=40}
 {bone1Name=dyn_pouch_l_02, bone2Name=dyn_pouch_l_01, swing1LimitDegrees=30, swing2LimitDegrees=20, twistLimitDegrees=20}
]
