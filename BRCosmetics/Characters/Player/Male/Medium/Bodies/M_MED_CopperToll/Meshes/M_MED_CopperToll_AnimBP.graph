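# Post-style cosmetic AnimBlueprint for the CopperToll male body, adding secondary motion to the pose copied from the attached parent mesh.
# Control Rig and a sleeve constraint feed RigidBody skirt sims (front-end or in-game), then cape sims (backbling variant adds bone offsets).
# AnimDynamics then springs ribbons, knot, ponytail, shirt and sleeves before fixed pouch offsets.
# Event Graph updates cape gravity via a deprecated function and detects backblings.
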AnimBlueprint /BRCosmetics/Characters/Player/Male/Medium/Bodies/M_MED_CopperToll/Meshes/M_MED_CopperToll_AnimBP
target_skeleton=&/M_MED_CopperToll_Skeleton
graphs [
 AnimGraph {nodes [
   n1 = "Transform (Modify) Bone - Bone: dyn_pouch_l_01"!(
    ComponentPose=Local To Component(LocalPose=Use cached pose 'postSimCache'())
    bAlphaBoolEnabled:bool=True
    Translation="-2,-2,0"
    Rotation="0,0,0"
    Scale="1,1,1"
   ) -> (Pose) {
    alphaInputType=Bool
    rotationMode=BMM_Additive
    rotationSpace=BCS_BoneSpace
    translationMode=BMM_Additive
    translationSpace=BCS_BoneSpace
    alphaBoolBlend {blendOption=Linear}
    boneToModify {boneName=dyn_pouch_l_01}
    translation=<-2 -2>
   }
   n2 = "Transform (Modify) Bone - Bone: dyn_pouch_l_02"!(ComponentPose=@n1, bAlphaBoolEnabled:bool=True, Translation="0,0,0", Rotation="15,0,0", Scale="1,1,1") -> (Pose) {
    alphaInputType=Bool
    rotationMode=BMM_Additive
    rotationSpace=BCS_BoneSpace
    translationMode=BMM_Additive
    translationSpace=BCS_BoneSpace
    alphaBoolBlend {blendOption=Linear}
    boneToModify {boneName=dyn_pouch_l_02}
    rotation {pitch=15}
   }
   Output Pose(Result=Component To Local(ComponentPose=@n2)) {layerGroup=None}
   n3 = Control Rig!(Source=Copy Pose From Mesh() {bUseAttachedParent=true}, Alpha, isParachuteOpen:bool=Get bIsParachuteOpened()) -> (Pose) {
    alphaBoolBlend {blendOption=Linear}
    controlRigAssetReference {blueprintRigClass=&/M_MED_DeformRig_CopperToll_ControlRig.M_MED_DeformRig_CopperToll_ControlRig_C}
   }
   Save cached pose 'CRCache'(Pose=@n3) {cacheName=CRCache}
   n4 = RigidBody!(ComponentPose=Local To Component(LocalPose=Use cached pose 'preSimCache'()), Alpha, OverrideWorldGravity:Vector="0,0,0", ExternalForce) -> (Pose) {
    bClampLinearTranslationLimitToRefPose=true
    bOverrideWorldGravity=true
    overridePhysicsAsset=&/M_MED_CopperToll_LowSkirt_Physics
    simulationSpace=BaseBoneSpace
    alphaBoolBlend {blendOption=Linear}
    baseBoneRef {boneName=pelvis}
    componentAppliedLinearAccClamp=<6000 6000 6000>
    componentLinearAccScale=<0.3 0.4 0.3>
    componentLinearVelScale=<0.3 0.4 0.3>
   }
   n5 = RigidBody(ComponentPose=Local To Component(LocalPose=Use cached pose 'preSimCache'()), OverrideWorldGravity="0,0,0") {
    bClampLinearTranslationLimitToRefPose=true
    bOverrideWorldGravity=true
    overridePhysicsAsset=&/M_MED_CopperToll_FrontEnd_LowSkirt_Physics
    simulationSpace=BaseBoneSpace
    alphaBoolBlend {blendOption=Linear}
    baseBoneRef {boneName=pelvis}
    componentAppliedLinearAccClamp=<6000 6000 6000>
    componentLinearAccScale=<0.3 0.4 0.3>
    componentLinearVelScale=<0.3 0.4 0.3>
   }
   Save cached pose 'RBANSkirtCache'(Pose=Blend Poses by bool(BlendPose_0=Component To Local(ComponentPose=@n5), BlendPose_1=Component To Local(ComponentPose=@n4), bActiveValue=Get bIsInFrontEnd()) {blendType=Linear}) {cacheName=RBANSkirtCache}
   n6 = "Constraint - dyn_capeSleeveA_r_MSTR (PP)"!(ComponentPose=Local To Component(LocalPose=Use cached pose 'CRCache'()), Alpha=1, ConstraintWeights_0=0.8, ConstraintWeights_1:float=0.2) -> (Pose) {
    alphaBoolBlend {blendOption=Linear}
    boneToModify {boneName=dyn_capeSleeveA_r_MSTR}
    constraintSetup [{targetBone {boneName=upperarm_twist_01_r}}, {targetBone {boneName=clavicle_r}}]
    constraintWeights [1,1]
   }
   Save cached pose 'preSimCache'(Pose=Component To Local(ComponentPose=@n6)) {cacheName=preSimCache}
   n7 = Anim Dynamics - Bone: dyn_smallRibbonA_l_01(ComponentPose=Local To Component(LocalPose=Use cached pose 'RBANCapeCache'()), Alpha=1, ExternalForce="0,0,0") {
    angularDampingOverride=1
    angularSpringConstant=20000
    bAngularSpring=true
    bOverrideAngularDamping=true
    bUseGravityOverride=true
    gravityScale=0.35
    simulationSpace=RootRelative
    alphaBoolBlend {blendOption=Linear}
    boundBone {boneName=dyn_smallRibbonA_l_01}
    componentLinearAccScale=<0.5 0.5 0.5>
    componentLinearVelScale=<0.5 0.5 0.5>
    gravityOverride {z=-980}
    physicsBodyDefinitions [{
      boundBone {boneName=dyn_smallRibbonA_l_01}
      boxExtents=<20 30 30>
      constraintSetup {angularLimitsMax=<15 15 15>, angularLimitsMin=<-15 -15 -15>, angularTarget=<1 0 0>}
      localJointOffset=<4 0 0>
     }]
   }
   n8 = Anim Dynamics - Bone: dyn_smallRibbonB_l_01(ComponentPose=@n7, Alpha=1, ExternalForce="0,0,0") {
    angularDampingOverride=1
    angularSpringConstant=20000
    bAngularSpring=true
    bOverrideAngularDamping=true
    bUseGravityOverride=true
    gravityScale=0.35
    simulationSpace=RootRelative
    alphaBoolBlend {blendOption=Linear}
    boundBone {boneName=dyn_smallRibbonB_l_01}
    componentLinearAccScale=<1 1 1>
    componentLinearVelScale=<1 1 1>
    gravityOverride {z=-980}
    physicsBodyDefinitions [{
      boundBone {boneName=dyn_smallRibbonB_l_01}
      boxExtents=<20 30 30>
      constraintSetup {angularLimitsMax=<15 15 15>, angularLimitsMin=<-15 -15 -15>, angularTarget=<1 0 0>}
      localJointOffset=<4 0 0>
     }]
   }
   n9 = Anim Dynamics - Bone: dyn_smallRibbonC_l_01(ComponentPose=@n8, Alpha=1, ExternalForce="0,0,0") {
    angularDampingOverride=1
    angularSpringConstant=20000
    bAngularSpring=true
    bOverrideAngularDamping=true
    bUseGravityOverride=true
    gravityScale=0.35
    simulationSpace=RootRelative
    alphaBoolBlend {blendOption=Linear}
    boundBone {boneName=dyn_smallRibbonC_l_01}
    componentLinearAccScale=<1 1 1>
    componentLinearVelScale=<1 1 1>
    gravityOverride {z=-980}
    physicsBodyDefinitions [{
      boundBone {boneName=dyn_smallRibbonC_l_01}
      boxExtents=<20 30 30>
      constraintSetup {angularLimitsMax=<15 15 15>, angularLimitsMin=<-15 -15 -15>, angularTarget=<1 0 0>}
      localJointOffset=<6 0 0>
     }]
   }
   n10 = Anim Dynamics - Bone: dyn_calfKnot_l_01(ComponentPose=@n9, Alpha=1, ExternalForce="0,0,0") {
    angularDampingOverride=1
    angularSpringConstant=20000
    bAngularSpring=true
    bOverrideAngularDamping=true
    bUseGravityOverride=true
    gravityScale=0.35
    simulationSpace=RootRelative
    alphaBoolBlend {blendOption=Linear}
    boundBone {boneName=dyn_calfKnot_l_01}
    componentLinearAccScale=<1 1 1>
    componentLinearVelScale=<1 1 1>
    gravityOverride {z=-980}
    physicsBodyDefinitions [{
      boundBone {boneName=dyn_calfKnot_l_01}
      boxExtents=<20 30 30>
      constraintSetup {angularLimitsMax=<15 15 15>, angularLimitsMin=<-15 -15 -15>, angularTarget=<1 0 0>}
      localJointOffset=<4 0 0>
     }]
   }
   n11 = Anim Dynamics - Bone: dyn_ponyTails_r_01(ComponentPose=@n10, Alpha=1, ExternalForce="0,0,0") {
    angularDampingOverride=1
    angularSpringConstant=20000
    bAngularSpring=true
    bOverrideAngularDamping=true
    bUseGravityOverride=true
    gravityScale=0.35
    simulationSpace=RootRelative
    alphaBoolBlend {blendOption=Linear}
    boundBone {boneName=dyn_ponyTails_r_01}
    componentLinearAccScale=<1 1 1>
    componentLinearVelScale=<1 1 1>
    gravityOverride {z=-980}
    physicsBodyDefinitions [{
      boundBone {boneName=dyn_ponyTails_r_01}
      boxExtents=<20 30 30>
      constraintSetup {angularLimitsMax=<15 15 15>, angularLimitsMin=<-15 -15 -15>, angularTarget=<1 0 0>}
      localJointOffset=<4 0 0>
     }]
   }
   n12 = Anim Dynamics - Bone: dyn_shirtBack_l_01(ComponentPose=@n11, Alpha=1, ExternalForce="0,0,0") {
    angularDampingOverride=1
    angularSpringConstant=20000
    bAngularSpring=true
    bLinearSpring=true
    bOverrideAngularDamping=true
    bUseGravityOverride=true
    gravityScale=0.35
    linearSpringConstant=30
    simulationSpace=RootRelative
    alphaBoolBlend {blendOption=Linear}
    boundBone {boneName=dyn_shirtBack_l_01}
    componentAppliedLinearAccClamp=<6000 6000 6000>
    componentLinearAccScale=<0.1 0.1 0.1>
    componentLinearVelScale=<0.1 0.1 0.1>
    gravityOverride {z=-980}
    physicsBodyDefinitions [{
      boundBone {boneName=dyn_shirtBack_l_01}
      boxExtents=<20 30 30>
      constraintSetup {angularLimitsMax=<0 20 20>, angularLimitsMin=<0 -20 -20>, angularTarget=<1 0 0>, linearAxesMax=<5 3 2>, linearAxesMin=<-5 -3 -2>}
      localJointOffset=<4 0 0>
     }]
    planarLimits [{drivingBone {boneName=None}, planeTransform {location=<0 0 0>, rotation=<rot 0 0 0>}}]
   }
   n13 = Anim Dynamics - Bone: dyn_shirtBack_r_01(ComponentPose=@n12, Alpha=1, ExternalForce="0,0,0") {
    angularDampingOverride=1
    angularSpringConstant=20000
    bAngularSpring=true
    bLinearSpring=true
    bOverrideAngularDamping=true
    bUseGravityOverride=true
    gravityScale=0.35
    linearSpringConstant=30
    simulationSpace=RootRelative
    alphaBoolBlend {blendOption=Linear}
    boundBone {boneName=dyn_shirtBack_r_01}
    componentAppliedLinearAccClamp=<6000 6000 6000>
    componentLinearAccScale=<0.3 0.3 0.3>
    componentLinearVelScale=<0.3 0.3 0.3>
    gravityOverride {z=-980}
    physicsBodyDefinitions [{
      boundBone {boneName=dyn_shirtBack_r_01}
      boxExtents=<20 30 30>
      constraintSetup {angularLimitsMax=<0 20 20>, angularLimitsMin=<0 -20 -20>, angularTarget=<1 0 0>, linearAxesMax=<5 3 2>, linearAxesMin=<-5 -3 -2>}
      localJointOffset=<4 0 0>
     }]
    planarLimits [{drivingBone {boneName=None}, planeTransform {location=<0 0 0>, rotation=<rot 0 0 0>}}]
   }
   n14 = Anim Dynamics - Bone: dyn_capeSleeveB_l_MSTR(ComponentPose=@n13, Alpha=1, ExternalForce="0,0,0") {
    angularDampingOverride=1
    angularSpringConstant=10000
    bAngularSpring=true
    bOverrideAngularDamping=true
    bUseGravityOverride=true
    gravityScale=0.35
    simulationSpace=RootRelative
    alphaBoolBlend {blendOption=Linear}
    boundBone {boneName=dyn_capeSleeveB_l_MSTR}
    componentLinearAccScale=<1 1 1>
    componentLinearVelScale=<1 1 1>
    gravityOverride {z=-980}
    physicsBodyDefinitions [{
      boundBone {boneName=dyn_capeSleeveB_l_MSTR}
      boxExtents=<20 30 30>
      constraintSetup {angularLimitsMax=<15 30 15>, angularLimitsMin=<-15 -15 -15>, angularTarget=<1 0 0>}
      localJointOffset=<6 0 0>
     }]
   }
   n15 = Anim Dynamics - Bone: dyn_capeSleeveA_l_MSTR(ComponentPose=@n14, Alpha=1, ExternalForce="0,0,0") {
    angularDampingOverride=1
    angularSpringConstant=10000
    bAngularSpring=true
    bOverrideAngularDamping=true
    bUseGravityOverride=true
    gravityScale=0.35
    simulationSpace=RootRelative
    alphaBoolBlend {blendOption=Linear}
    boundBone {boneName=dyn_capeSleeveA_l_MSTR}
    componentLinearAccScale=<1 1 1>
    componentLinearVelScale=<1 1 1>
    gravityOverride {z=-980}
    physicsBodyDefinitions [{
      boundBone {boneName=dyn_capeSleeveA_l_MSTR}
      boxExtents=<20 30 30>
      constraintSetup {angularLimitsMax=<15 0 15>, angularLimitsMin=<-15 -15 -15>, angularTarget=<1 0 0>}
      localJointOffset=<6 0 0>
     }]
   }
   n16 = Anim Dynamics - Bone: dyn_pouch_l_02(ComponentPose=@n15, Alpha=0, ExternalForce="0,0,0") {
    angularDampingOverride=1
    angularSpringConstant=20000
    bAngularSpring=true
    bOverrideAngularDamping=true
    bUseGravityOverride=true
    gravityScale=0.35
    simulationSpace=RootRelative
    alphaBoolBlend {blendOption=Linear}
    boundBone {boneName=dyn_pouch_l_02}
    componentLinearAccScale=<1 1 1>
    componentLinearVelScale=<1 1 1>
    gravityOverride {z=-980}
    physicsBodyDefinitions [{
      boundBone {boneName=dyn_pouch_l_02}
      boxExtents=<20 30 30>
      constraintSetup {angularLimitsMax=<15 15 15>, angularLimitsMin=<-15 -15 -15>, angularTarget=<1 0 0>}
      localJointOffset=<4 0 0>
     }]
   }
   Save cached pose 'postSimCache'(Pose=Component To Local(ComponentPose=@n16)) {cacheName=postSimCache}
   n17 = RigidBody(ComponentPose=Local To Component(LocalPose=Use cached pose 'RBANSkirtCache'()), OverrideWorldGravity=Get Gravity Override()) {
    bClampLinearTranslationLimitToRefPose=true
    bOverrideWorldGravity=true
    overridePhysicsAsset=&/M_MED_CopperToll_Cape_Backbling_Physics
    simulationSpace=BaseBoneSpace
    alphaBoolBlend {blendOption=Linear}
    baseBoneRef {boneName=spine_05}
    componentAppliedLinearAccClamp=<6000 6000 6000>
    componentLinearAccScale=<0.1 0.1 0.1>
    componentLinearVelScale=<0.1 0.1 0.1>
    overrideWorldGravity {z=-980}
    simSpaceSettings {worldAlpha=0.35, externalAngularVelocity {x=0.2, z=0.2}, externalLinearVelocity {x=0.2, z=0.2}}
   }
   n18 = "Constraint - dyn_capeSleeveBack_c_02 (PP)"!(ComponentPose=@n17, Alpha=1, ConstraintWeights_0=0.7, ConstraintWeights_1:float=0.3) -> (Pose) {
    alphaBoolBlend {blendOption=Linear}
    boneToModify {boneName=dyn_capeSleeveBack_c_02}
    constraintSetup [{targetBone {boneName=spine_04}}, {targetBone {boneName=spine_05}}]
    constraintWeights [1,1]
   }
   n19 = "Transform (Modify) Bone - Bone: dyn_capeSleeveBack_c_MSTR"(ComponentPose=@n18, Alpha=1, Translation="0,0,-2", Rotation="-3,-4,0", Scale="1,1,1") {
    rotationMode=BMM_Additive
    rotationSpace=BCS_BoneSpace
    translationMode=BMM_Additive
    translationSpace=BCS_BoneSpace
    alphaBoolBlend {blendOption=Linear}
    boneToModify {boneName=dyn_capeSleeveBack_c_MSTR}
    rotation {pitch=-3, yaw=-4}
    translation {z=-2}
   }
   n20 = RigidBody(ComponentPose=Local To Component(LocalPose=Use cached pose 'RBANSkirtCache'()), OverrideWorldGravity=Get Gravity Override()) {
    bClampLinearTranslationLimitToRefPose=true
    bOverrideWorldGravity=true
    overridePhysicsAsset=&/M_MED_CopperToll_Cape_Physics
    simulationSpace=BaseBoneSpace
    alphaBoolBlend {blendOption=Linear}
    baseBoneRef {boneName=spine_05}
    componentAppliedLinearAccClamp=<6000 6000 6000>
    componentLinearAccScale=<0.1 0.1 0.1>
    componentLinearVelScale=<0.1 0.1 0.1>
    overrideWorldGravity {z=-980}
    simSpaceSettings {worldAlpha=0.35, externalAngularVelocity {x=0.2, z=0.2}, externalLinearVelocity {x=0.2, z=0.2}}
   }
   Save cached pose 'RBANCapeCache'(Pose=Blend Poses by bool(
     BlendPose_0=Component To Local(ComponentPose=@n19)
     BlendPose_1=Component To Local(ComponentPose=@n20)
     bActiveValue=Get Is Wearing Backbling()
    ) {blendType=Linear}) {cacheName=RBANCapeCache}
  ]}
 EventGraph {nodes [
   n1 = Event BlueprintUpdateAnimation()
   > n2 = Sequence!(execute) -> (then_0, then_1, then_2:Exec)
   > n3 = Gravity Override State Function DEPRECATED(
    CharacterPartAnimInstance=Self-Reference()
    GameTimeInSeconds=0
    GravityOverrideParams=Get Gravity Override Params_0() -> (Gravity Override Params_0:Gravity Override Params Struct Structure)
   )
   > Set Gravity Override(Gravity Override=@n3)
   n4 = CheckForBackbling(self=Self-Reference())
   n5 = Branch(execute=@n2.then_1, Condition=OR Boolean!(A=@n4, B=@n4.bIsBackblingACape, C:bool=@n4.bIsBackblingLong) -> (ReturnValue))
   > Set Is Wearing Backbling(Is Wearing Backbling=true)
   Set Is Wearing Backbling(execute=@n5.else, Is Wearing Backbling=false)
   Set Delta Time X(Delta Time X=@n1.DeltaTimeX)
  ]}
]
variables [
 Delta Time X {type=number}
 Gravity Override {type=Vector}
 Gravity Override Params {type=GravityOverrideParamsStruct}
 Gravity Override Params_0 {type=GravityOverrideParamsStruct}
 Is Wearing Backbling {type=boolean}
 NewVar {type=SkeletalMeshComponent}
]
