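# Post-process style AnimBP for the CopperToll male medium face accessory, correcting hood, cheek and head bones on the attached parent's copied pose.
# A chain of additive bone-space Modify Bone nodes applies fixed offsets at constant full alpha.
# The update event derives head-versus-spine_05 rotation and twist alphas from the body part mesh, but the AnimGraph reads none.
# Its positive-range twist overwrites headTwistNeg, leaving headTwistPos unset.
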AnimBlueprint /BRCosmetics/Characters/Player/Male/Medium/Bodies/M_MED_CopperToll/Meshes/Parts/M_MED_CopperToll_FaceAcc_AnimBP
target_skeleton=&/M_MED_CopperToll_FaceAcc_Skeleton
graphs [
 AnimGraph {nodes [
   n1 = "Transform (Modify) Bone - Bone: dyn_hoodBack_01"(
    ComponentPose=Local To Component(LocalPose=Copy Pose From Mesh() {bUseAttachedParent=true})
    Alpha=1
    Translation="-6,-6,0"
    Rotation="0,0,0"
    Scale="1,1,1"
   ) {
    rotationMode=BMM_Additive
    rotationSpace=BCS_BoneSpace
    translationMode=BMM_Additive
    translationSpace=BCS_BoneSpace
    alphaBoolBlend {blendOption=Linear}
    boneToModify {boneName=dyn_hoodBack_01}
    translation=<-6 -6>
   }
   n2 = "Transform (Modify) Bone - Bone: dyn_hoodFront_l_MSTR"(ComponentPose=@n1, Alpha=1, Translation="0,-4,0", Rotation="0,-10,0", Scale="1,1,1") {
    rotationMode=BMM_Additive
    rotationSpace=BCS_BoneSpace
    translationMode=BMM_Additive
    translationSpace=BCS_BoneSpace
    alphaBoolBlend {blendOption=Linear}
    boneToModify {boneName=dyn_hoodFront_l_MSTR}
    rotation {yaw=-10}
    translation {y=-4}
   }
   n3 = "Transform (Modify) Bone - Bone: dyn_hoodFront_r_MSTR"(ComponentPose=@n2, Alpha=1, Translation="0,-4,6", Rotation="0,-10,0", Scale="1,1,1") {
    rotationMode=BMM_Additive
    rotationSpace=BCS_BoneSpace
    translationMode=BMM_Additive
    translationSpace=BCS_BoneSpace
    alphaBoolBlend {blendOption=Linear}
    boneToModify {boneName=dyn_hoodFront_r_MSTR}
    rotation {yaw=-10}
    translation {y=-4, z=6}
   }
   n4 = "Transform (Modify) Bone - Bone: dyn_hoodBack_01"(ComponentPose=@n3, Alpha=1, Translation="-2,-3,0", Rotation="0,0,0", Scale="1,1,1") {
    rotationMode=BMM_Additive
    rotationSpace=BCS_BoneSpace
    translationMode=BMM_Additive
    translationSpace=BCS_BoneSpace
    alphaBoolBlend {blendOption=Linear}
    boneToModify {boneName=dyn_hoodBack_01}
    translation=<-2 -3>
   }
   n5 = "Transform (Modify) Bone - Bone: dyn_hoodFront_l_MSTR"(ComponentPose=@n4, Alpha=1, Translation="-2,0,0", Rotation="0,0,0", Scale="1,1,1") {
    rotationMode=BMM_Additive
    rotationSpace=BCS_BoneSpace
    translationMode=BMM_Additive
    translationSpace=BCS_BoneSpace
    alphaBoolBlend {blendOption=Linear}
    boneToModify {boneName=dyn_hoodFront_l_MSTR}
    translation {x=-2}
   }
   n6 = "Transform (Modify) Bone - Bone: dyn_hoodFront_r_MSTR"(ComponentPose=@n5, Alpha=1, Translation="0,0,0", Rotation="0,-10,0", Scale="1,1,1") {
    rotationMode=BMM_Additive
    rotationSpace=BCS_BoneSpace
    translationMode=BMM_Additive
    translationSpace=BCS_BoneSpace
    alphaBoolBlend {blendOption=Linear}
    boneToModify {boneName=dyn_hoodFront_r_MSTR}
    rotation {yaw=-10}
   }
   n7 = "Transform (Modify) Bone - Bone: dyn_cheek_l_01"(ComponentPose=@n6, Alpha=1, Translation="0,0,2.5", Rotation="0,0,0", Scale="1,1,1") {
    rotationMode=BMM_Additive
    rotationSpace=BCS_BoneSpace
    translationMode=BMM_Additive
    translationSpace=BCS_BoneSpace
    alphaBoolBlend {blendOption=Linear}
    boneToModify {boneName=dyn_cheek_l_01}
    translation {z=2.5}
   }
   n8 = "Transform (Modify) Bone - Bone: dyn_hoodBack_01"(ComponentPose=@n7, Alpha=1, Translation="-2,-3,0", Rotation="0,0,0", Scale="1,1,1") {
    rotationMode=BMM_Additive
    rotationSpace=BCS_BoneSpace
    translationMode=BMM_Additive
    translationSpace=BCS_BoneSpace
    alphaBoolBlend {blendOption=Linear}
    boneToModify {boneName=dyn_hoodBack_01}
    translation=<-2 -3>
   }
   n9 = "Transform (Modify) Bone - Bone: dyn_hoodFront_l_MSTR"(ComponentPose=@n8, Alpha=1, Translation="0,0,0", Rotation="0,0,0", Scale="1,1,1") {
    rotationMode=BMM_Additive
    rotationSpace=BCS_BoneSpace
    translationMode=BMM_Additive
    translationSpace=BCS_BoneSpace
    alphaBoolBlend {blendOption=Linear}
    boneToModify {boneName=dyn_hoodFront_l_MSTR}
   }
   n10 = "Transform (Modify) Bone - Bone: dyn_hoodFront_r_MSTR"(ComponentPose=@n9, Alpha=1, Translation="0,0,-4", Rotation="0,0,0", Scale="1,1,1") {
    rotationMode=BMM_Additive
    rotationSpace=BCS_BoneSpace
    translationMode=BMM_Additive
    translationSpace=BCS_BoneSpace
    alphaBoolBlend {blendOption=Linear}
    boneToModify {boneName=dyn_hoodFront_r_MSTR}
    translation {z=-4}
   }
   n11 = "Transform (Modify) Bone - Bone: head"(ComponentPose=@n10, Alpha=1, Translation="0,0,0", Rotation="0,0,0", Scale="1,1,1.01") {
    rotationMode=BMM_Additive
    rotationSpace=BCS_BoneSpace
    scaleMode=BMM_Additive
    scaleSpace=BCS_BoneSpace
    translationMode=BMM_Additive
    translationSpace=BCS_BoneSpace
    alphaBoolBlend {blendOption=Linear}
    boneToModify {boneName=head}
    scale {z=1.01}
   }
   Output Pose(Result=Component To Local(ComponentPose=@n11)) {layerGroup=None}
  ]}
 EventGraph {nodes [
   Event BlueprintUpdateAnimation()
   > n1 = Sequence!(execute) -> (then_0, then_1, then_2:Exec, then_3:Exec)
   > Set Body(Body:Skeletal Mesh Component Object Reference=GetPartSkeletalMeshComponent()) -> (Output_Get:Skeletal Mesh Component Object Reference)
   Set headRotZNeg(
    execute=@n1.then_1
    headRotZNeg:float=GetAlphaFromReferenceTransformDelta(
     SkeletalMeshComponent=Get Body() -> (Body:Skeletal Mesh Component Object Reference)
     BoneName=head
     BaseName=spine_05
     InputRange=(X=0,Y=-40)
     TwistAxis="1,0,0"
    )
   ) -> (Output_Get:float)
   Set headTwistNeg(
    execute=@n1.then_2
    headTwistNeg:float=GetAlphaFromReferenceTransformDelta(
     SkeletalMeshComponent=Get Body()
     BoneName=head
     BaseName=spine_05
     DeltaComponentType=QuaternionTwist
     InputRange=(X=0,Y=-40)
     TwistAxis="1,0,0"
    )
   ) -> (Output_Get:float)
   Set headTwistNeg(
    execute=@n1.then_3
    headTwistNeg=GetAlphaFromReferenceTransformDelta(
     SkeletalMeshComponent=Get Body()
     BoneName=head
     BaseName=spine_05
     DeltaComponentType=QuaternionTwist
     InputRange=(X=0,Y=40)
     TwistAxis="1,0,0"
    )
   )
  ]}
]
variables [
 Body {type=SkeletalMeshComponent}
 headRotZNeg {type=number}
 headTwistNeg {type=number}
 headTwistPos {type=number}
]
